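# Niagara ribbon smoke trail for the GameBreaker player emote, named for battery trails and built from a Reload_Smoke emitter.
# Ribbons spawn at a steady rate with slight upward velocity and curl noise, their lifetime shrinking over the emitter loop.
# Alpha fades with age, loop and time of day, while user parameters control color, width, lifetime and optional emissive tint.
# Skeletal mesh location and source-movement inheritance modules are disabled.
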NiagaraSystem /BRCosmetics/Animation/Game/MainPlayer/Emotes/GameBreaker/FX/NS_Battary_Trails
dependencies {
 usedDataInterfaces [&NiagaraDataInterfaceCurve, &NiagaraDataInterfaceSkeletalMesh]
 usedDynamicInputs []
 usedModules [
  &self:SystemSpawnScript.SystemScriptSource.SystemScriptGraph.NiagaraNodeAssignment_0.SetVariables_#1
  &SystemState
 ]
}
emitters {Reload_Smoke {
  emitterUpdateScript [
   &EmitterState()
   &SpawnRate(SpawnRate=90, Use Spawn Probability=false)
  ]
  particleSpawnScript [
   &InitializeParticle(
    Color=$User.BaseColorAlpha
    Lifetime=&FloatFromCurve(
     FloatCurve=DataInterface() {propertyValues {Curve {keys [
         <at 0 0.5 RCIM_Cubic>
         {arriveTangent=-0.3345, interpMode=RCIM_Cubic, leaveTangent=-0.3102, tangentMode=RCTM_Break, time=0.5, value=0.45}
         {arriveTangent=-1.2344, interpMode=RCIM_Cubic, leaveTangent=-1.2344, tangentMode=RCTM_User, time=0.8, value=0}
        ]}}}
     CurveIndex=$Emitter.NormalizedLoopAge
     Scale Curve=$User.RibbonLifetime
    )
    Ribbon Width=&Multiply_Float(A=10.7, B=$User.RibbonScale)
    Ribbon Width Mode=::NewEnumerator1
    UsePositionOffset=false
    Write Material Random=true
   )
   &AddVelocity(
    Coordinate Space=::World
    Velocity=&MakeVector(
     X=0
     Y=0
     Z=&Multiply_Float(
      A=1
      B=&FloatFromCurve(FloatCurve=DataInterface() {propertyValues {Curve {keys [<at 0.5 0.4>,<at 1 0>]}}}, CurveIndex=$Emitter.NormalizedLoopAge)
     )
    )
   )
   &SkeletalMeshLocation(MaskSpawnGroup=false, OverrideRandomSeed=false) {enabled=false}
   &SkeletalMeshLocation(MaskSpawnGroup=false, OverrideRandomSeed=false) {enabled=false, module=SkeletalMeshLocation001}
  ]
  particleUpdateScript [
   &ParticleState()
   &ScaleColor(
    Scale Alpha=&Multiply_Float(
     A=&FloatFromCurve(
      FloatCurve=DataInterface() {propertyValues {Curve {keys [<at 0 1>, {arriveTangent=-1.7714, interpMode=RCIM_Cubic, leaveTangent=-1.7714, tangentMode=RCTM_User, time=0.6, value=0.8}, <at 1 0>]}}}
      Scale Curve=&FNTimeOfDayScalar(Night Scale=1, Morning Scale=0.3, Evening Scale=0.5)
     )
     B=&FloatFromCurve(
      FloatCurve=DataInterface() {propertyValues {Curve {keys [<at 0 1>, {arriveTangent=-1, interpMode=RCIM_Cubic, leaveTangent=-1, time=0.5, value=1}, <at 1 0 RCIM_Cubic>]}}}
      CurveIndex=$Emitter.NormalizedLoopAge
     )
    )
    Scale RGB=&MakeVectorFromLinearColorRGB(Color=&MakeCustomLinearColorFromBool(
      Bool=$User.UseEmissiveSmoke
      True Color=&LerpLinearColors(
       StartColor=$User.EmissiveColor
       EndColor=&MultiplyLinearColor(Linear Color=$User.EmissiveColor, Scale Factor=&MakeVector4_FromFloat(Float=0.25))
       LerpFactor=&FloatFromCurve(FloatCurve=DataInterface() {propertyValues {Curve {keys [<at 0 0 RCIM_Cubic>, {arriveTangent=2.2656, interpMode=RCIM_Cubic, leaveTangent=2.2656, tangentMode=RCTM_User, time=1, value=1}]}}})
      )
      False Color=<rgba 1 1 1 1>
     ))
    ScaleA=true
    ScaleRGB=true
   )
   &DynamicMaterialParameters(
    Index 0 Param 1=&FloatFromCurve(FloatCurve=DataInterface() {propertyValues {Curve {keys [
         <at 0 1 RCIM_Cubic>
         {arriveTangent=-1.7387, arriveTangentWeight=1.9926, interpMode=RCIM_Cubic, leaveTangent=-1.7387, tangentMode=RCTM_User, time=1, value=0.125}
        ]}}})
    Index 0 Param 3=-3
    Index 0 Param 4=&MakeCustomFloatFromBool(
     Bool=$User.UseEmissiveSmoke
     True Float=&FloatFromCurve(FloatCurve=DataInterface() {propertyValues {Curve {keys [{arriveTangent=-2.94, interpMode=RCIM_Cubic, time=0, value=1}, <at 0.75 0 RCIM_Cubic>]}}})
     False Float=0
    )
    Index 1 Param 1=&FloatFromCurve(FloatCurve=DataInterface() {propertyValues {Curve {keys [<at 0 1 RCIM_Cubic>, {arriveTangent=45.3807, interpMode=RCIM_Cubic, leaveTangent=45.3807, tangentMode=RCTM_User, time=0.5, value=10}]}}})
    Index 1 Param 2=&FloatFromCurve(FloatCurve=DataInterface() {propertyValues {Curve {keys [{arriveTangent=-2.2405, interpMode=RCIM_Cubic, leaveTangent=-2.2405, tangentMode=RCTM_User, time=0, value=1}, <at 1 0 RCIM_Cubic>]}}})
    Module.Param0WriteEnabled=true
    Module.Param0WriteEnabled1=true
    Module.Param1WriteEnabled=false
    Module.Param1WriteEnabled1=true
    Module.Param2WriteEnabled=true
    Module.Param2WriteEnabled1=false
    Module.Param3WriteEnabled=true
    Module.Param3WriteEnabled1=false
    Write Parameter Index 1=true
   )
   &ScaleRibbonWidth(Ribbon Width Scale=&FloatFromCurve(FloatCurve=DataInterface() {propertyValues {Curve {keys [{tangentMode=RCTM_User, time=0, value=0.5875}, <at 0.1 0.5>, {tangentMode=RCTM_User, time=1, value=0.125}]}}}))
   &CurlNoiseForce(
    Mask Curl Noise=true
    Noise Frequency=10
    Noise Quality / Cost=::NewEnumerator0
    Noise Strength=725
    Pan Noise=true
    Pan Noise Field=<-10 0 0>
   )
   &InheritSourceMovement(
    Applied Position Scale=&Multiply_VectorByFloat(
     Vector=<1 1 1>
     Float=&FloatFromCurve(FloatCurve=DataInterface() {propertyValues {Curve {keys [<at 0 1>,<at 1 0.6>]}}})
    )
    ApplyForce=false
    ApplyPosition=true
    ApplyPositionFalloff=false
   ) {enabled=false}
   &SolveForcesAndVelocity(Clamp Velocity=false, Limit Acceleration=false)
  ]
  rendererClasses [&NiagaraRibbonRendererProperties]
  renderers [{rendererClass=&NiagaraRibbonRendererProperties}]
 }}
summary {
 systemName=NS_Battary_Trails
 emitters [{emitterName=Reload_Smoke, rendererClasses [&NiagaraRibbonRendererProperties]}]
 userVariables [
  User.BaseColorAlpha {type=LinearColor, description="", defaultValue {struct=LinearColor, value=<rgba 1 1 1 0.15>}}
  User.EmissiveColor {type=LinearColor, description="", defaultValue {struct=LinearColor, value=<rgba 0.1354 0.0043 0 1>}}
  User.HasShellEjects {type=NiagaraBool}
  User.RibbonLifetime {type=NiagaraFloat, defaultValue=0.5, description=""}
  User.RibbonMaterial {
   type=MaterialInterface
   description=""
   defaultValue {
    struct=NiagaraExt_VariableValue_Object
    value {object=/Game/Effects/Niagara/MasterMaterials/Reload/M_Reload_SmokeTrail, objectClass=/Script/Engine.MaterialInterface}
   }
  }
  User.RibbonScale {type=NiagaraFloat, defaultValue=1, description=""}
  User.UseEmissiveSmoke {type=NiagaraBool}
 ]
}
