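# NS_BigBurst is a Niagara burst effect for the GameBreaker emote, combining a flash, sparks and streaks in mostly green tones.
# Flash_Flare001 spawns one brief user-colored sprite with a light, while Streaks_Fat throws stretching streaks outward after User.ExplosionDelay.
# NE_Sprites fires two drag-slowed spark groups, writing extra color and size attributes for its three sprite renderers, including blue-cyan variation.
# System ExplosionScale scales spark and streak speeds.
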NiagaraSystem /BRCosmetics/Animation/Game/MainPlayer/Emotes/GameBreaker/FX/NS_BigBurst
dependencies {
 usedDataInterfaces [&NiagaraDataInterfaceVector2DCurve, &NiagaraDataInterfaceColorCurve, &NiagaraDataInterfaceCurve]
 usedDynamicInputs [
  &SimulationPosition
  &Add_Float
  &SetBoolByFloatComparison
 ]
 usedModules [
  &self:SystemSpawnScript.SystemScriptSource.SystemScriptGraph.NiagaraNodeAssignment_3.SetVariables_#1
  &self:SystemSpawnScript.SystemScriptSource.SystemScriptGraph.NiagaraNodeAssignment_0.SetVariables_#2
  &self:SystemSpawnScript.SystemScriptSource.SystemScriptGraph.NiagaraNodeAssignment_6.SetVariables_#3
  &SystemState
  &self:SystemSpawnScript.SystemScriptSource.SystemScriptGraph.NiagaraNodeAssignment_5.SetVariables_#4
 ]
}
emitters {
 Flash_Flare001 {
  emitterUpdateScript [
   &EmitterState(Life Cycle Mode=::NewEnumerator1, Loop Duration=$User.LoopDuration, UseLoopDelay=false)
   &SpawnBurst_Instantaneous(Spawn Count=1, Use Spawn Probability=false)
  ]
  particleSpawnScript [
   &InitializeParticle(
    Color=$User.Color
    Lifetime=0.1
    Sprite Size=<250 250>
    Sprite Size Mode=::NewEnumerator4
    UsePositionOffset=false
    Write Material Random=true
   )
   &CameraOffset(Camera Offset Amount=50) {enabled=false}
   &DynamicMaterialParameters(
    Index 0 Param 1=0.2
    Index 0 Param 2=1
    Module.Param0WriteEnabled=true
    Module.Param1WriteEnabled=true
    Module.Param2WriteEnabled=true
    Module.Param3WriteEnabled=true
   )
   &SystemLocation(MaskSpawnGroup=false, Offset=$System.HitOffset)
  ]
  particleUpdateScript [
   &ParticleState()
   &ScaleColor(Scale Alpha=&Lerp_Float(A=1, B=0, Alpha=$Particles.NormalizedAge), ScaleA=true, ScaleRGB=true)
   &FN_Light_v2(Light Exponent=3, Light Radius=150, LightEnabled=true, Scale Color=1)
   &CameraOffset(Camera Offset Amount=15) {module=CameraOffset001}
   &Light_Attributes(Position=1, UsePositionOffset=false)
  ]
  rendererClasses [&NiagaraSpriteRendererProperties, &NiagaraLightRendererProperties]
  renderers [
   {rendererClass=&NiagaraSpriteRendererProperties}
   {rendererClass=&NiagaraLightRendererProperties, rendererIndex=1}
  ]
 }
 NE_Sprites {
  emitterUpdateScript [
   &EmitterState()
   &SpawnBurst_Instantaneous(Spawn Count=30, Use Spawn Probability=false, UseLoopCountLimit=false)
   &SpawnBurst_Instantaneous(Spawn Count=30, Spawn Group=1, Use Spawn Probability=false, UseLoopCountLimit=false) {module=SpawnBurst_Instantaneous001}
  ]
  particleSpawnScript [
   &InitializeParticle(
    Lifetime=&RandomRangeFloat(Minimum=0.3, Maximum=0.4)
    Sprite Size=&Multiply_Vector2D_ByFloat(Float=&RandomRangeFloat(Minimum=1, Maximum=2), Vector2D=<2 6>)
    Sprite Size Mode=::NewEnumerator4
    UsePositionOffset=false
    Write Material Random=true
   )
   /SetVariables(Particles.VisibilityTag=&SpawnGroup()) {isSetParameters=true}
   &ShapeLocation(
    MaskSpawnGroup=false
    Override Seed=true
    Random Seed:NiagaraInt32=&SpawnGroup()
    Sphere Radius=1
    UseOwnerScale=true
   )
   &AddVelocity(
    Velocity=$Particles.ShapeLocation.ShapeNormal
    Velocity Speed Scale=&Multiply_Float(A=&RandomRangeFloat(Minimum=100, Maximum=250), B=$System.ExplosionScale)
   )
  ]
  particleUpdateScript [
   &ParticleState()
   &Drag(Drag=&RandomRangeFloat(Minimum=4, Maximum=5), Use Linear Drag=true, Use Rotational Drag=true)
   &ScaleColor(
    Scale Alpha=&FloatFromCurve(FloatCurve=DataInterface() {propertyValues {Curve {keys [
         {arriveTangent=6.4019, interpMode=RCIM_Cubic, leaveTangent=6.4019, tangentMode=RCTM_User, time=0, value=0}
         <at 0.1 0.7 RCIM_Cubic>
         <at 1 0 RCIM_Cubic>
        ]}}})
    Scale RGB=&MakeVectorFromLinearColorRGB(Color=&MultiplyLinearColor(
      Linear Color=&ColorFromCurve(ColorCurve=DataInterface() {propertyValues {
         AlphaCurve {keys [<at 0 1 RCIM_Cubic>]}
         BlueCurve {keys [<at 0 10 RCIM_Cubic>,<at 0.3726 1 RCIM_Cubic>]}
         GreenCurve {keys [<at 0 10 RCIM_Cubic>,<at 0.3726 1 RCIM_Cubic>]}
         RedCurve {keys [<at 0 10 RCIM_Cubic>,<at 0.3726 1 RCIM_Cubic>]}
        }})
      Scale Factor=&MakeVector4_FromLinearColor(LinearColor=&ColorFromCurve(
        ColorCurve=DataInterface() {propertyValues {
          AlphaCurve {keys [<at 0 1 RCIM_Cubic>]}
          BlueCurve {keys [<at 0.3173 0 RCIM_Cubic>,<at 0.3548 0>,<at 0.6928 0.0767>,<at 0.7245 0.0645>]}
          GreenCurve {keys [<at 0.3173 1 RCIM_Cubic>,<at 0.3548 1>,<at 0.6928 1>,<at 0.7245 1>]}
          RedCurve {keys [<at 0.3173 0.1148 RCIM_Cubic>,<at 0.3548 0.0441>,<at 0.6928 0.013>,<at 0.7245 0>]}
         }}
        CurveIndex=&RandomRangeFloat(Minimum=0, Maximum=1)
       ))
     ))
    ScaleA=true
    ScaleRGB=true
   )
   &InheritSourceMovement(
    Applied Position Scale=&VectorFromFloat(Value=&FloatFromCurve(
      FloatCurve=DataInterface() {propertyValues {Curve {keys [
          {arriveTangent=2.6572, interpMode=RCIM_Cubic, leaveTangent=2.6572, tangentMode=RCTM_User, time=0, value=0}
          {interpMode=RCIM_Cubic, tangentMode=RCTM_User, time=1, value=1}
         ]}}}
      Scale Curve=-0.2
     ))
    ApplyForce=false
    ApplyPosition=true
    ApplyPositionFalloff=false
   )
   &SolveForcesAndVelocity(Clamp Velocity=false, Limit Acceleration=false)
   &DynamicMaterialParameters(
    Index 0 Param 1=2
    Index 0 Param 3=1
    Index 3 Param 1=&Lerp_Float(
     A=0.2
     B=1
     Alpha=&FloatFromCurve(FloatCurve=DataInterface() {propertyValues {Curve {keys [<at 0 0>,<at 1 1>]}}})
    )
    Module.Index0Float0Write=true
    Module.Index0Float1Write=true
    Module.Index0Float2Write=true
    Module.Index0Float3Write=true
    Module.Index3Float0Write=true
    Module.Index3Float1Write=true
    Module.Index3Float2Write=true
    Module.Index3Float3Write=true
    Write Parameter Index 3=::NewEnumerator1
   )
   /SetVariables(Particles.SparkGlowColor:LinearColor=&MultiplyLinearColor(
     Linear Color=<rgba 0.1 0.1 0.1 0.05>
     Scale Factor=&MakeVector4_FromLinearColor(LinearColor=$Particles.Color)
    )) {isSetParameters=true}
   /SetVariables(Particles.SparkGlowSize:Vector2f=&Multiply_Vector2D(A=$Particles.SpriteSize, B=<1 0.5>)) {isSetParameters=true}
   /SetVariables(Particles.SparkGlowDynMat:Vector4f=<2 20 90 1>) {isSetParameters=true}
   &FN_Sprite_Attributes(
    Color=&MakeLinearColorFromVectorAndFloat(
     "Vector (RGB)"=&MakeVectorFromLinearColorRGB(Color=&Multiply_LinearColorByFloat(
       Linear Color=&ColorFromCurve(
        ColorCurve=DataInterface() {propertyValues {
          AlphaCurve {keys [<at 0 1 RCIM_Cubic>]}
          BlueCurve {keys [<at 0.3173 1 RCIM_Cubic>,<at 0.3548 1>,<at 0.6928 1>,<at 0.7245 1>]}
          GreenCurve {keys [<at 0.3173 0 RCIM_Cubic>,<at 0.3548 0.3419>,<at 0.6928 0.3419>,<at 0.7245 0.9447>]}
          RedCurve {keys [<at 0.3173 0.0183 RCIM_Cubic>,<at 0.3548 0>,<at 0.6928 0>,<at 0.7245 0.0002>]}
         }}
        CurveIndex=&RandomRangeFloat(Minimum=0, Maximum=1)
       )
       Float=&Lerp_Float(
        A=1
        B=20
        Alpha=&FloatFromCurve(FloatCurve=DataInterface() {propertyValues {Curve {keys [
             {arriveTangent=-0.4493, interpMode=RCIM_Cubic, leaveTangent=-0.4493, tangentMode=RCTM_User, time=0, value=1}
             {interpMode=RCIM_Cubic, tangentMode=RCTM_User, time=0.7, value=0}
            ]}}})
       )
      ))
     "Float (Alpha)"=&FloatFromCurve(FloatCurve=DataInterface() {propertyValues {Curve {keys [
          {interpMode=RCIM_Cubic, tangentMode=RCTM_User, time=0, value=1}
          {arriveTangent=-2.0246, interpMode=RCIM_Cubic, leaveTangent=-2.0246, tangentMode=RCTM_User, time=1, value=0}
         ]}}})
    )
    Create Color Attribute=true
    Create SpriteSize Attribute=true
    SpriteSize=&Multiply_Vector2D_ByFloat(
     Float=&Multiply_Float(
      A=&RandomRangeFloat(Minimum=0.5, Maximum=1.5)
      B=&FloatFromCurve(FloatCurve=DataInterface() {propertyValues {Curve {keys [
           {interpMode=RCIM_Cubic, tangentMode=RCTM_User, time=0.2, value=1}
           {arriveTangent=-2.4619, interpMode=RCIM_Cubic, leaveTangent=-2.4619, tangentMode=RCTM_User, time=1, value=0}
          ]}}})
     )
     Vector2D=<2 2>
    )
    Write to Module Namespace=true
   )
  ]
  rendererClasses [&NiagaraSpriteRendererProperties]
  renderers [
   {rendererClass=&NiagaraSpriteRendererProperties}
   {rendererClass=&NiagaraSpriteRendererProperties, rendererIndex=1}
   {rendererClass=&NiagaraSpriteRendererProperties, rendererIndex=2}
  ]
 }
 Streaks_Fat {
  emitterUpdateScript [
   &EmitterState(Inactive Response=::NewEnumerator2, Life Cycle Mode=::NewEnumerator1, Loop Behavior=::NewEnumerator1, UseLoopDelay=false)
   &SpawnBurst_Instantaneous(
    Spawn Count=&RandomRangeInt(Minimum=15, Maximum=20)
    Spawn Time=$User.ExplosionDelay
    Use Spawn Probability=false
   )
  ]
  particleSpawnScript [
   &InitializeParticle(
    Lifetime=&UniformRangedFloat(Minimum=0.15, Maximum=0.25)
    Mass Mode=::NewEnumerator1
    Position Offset=$System.SystemOffset
    Sprite Size=&Multiply_Vector2D_ByFloat(Float=1, Vector2D=&UniformRangedVector2D(Minimum=<1 12>, Maximum=<3 15>))
    Sprite Size Mode=::NewEnumerator4
    UsePositionOffset=true
    Write Material Random=true
   ) {module=InitializeParticle001}
   /Niagara/Modules/Spawn/Location/SphereLocation(
    MaskSpawnGroup=false
    Sphere Distribution:ENiagaraTorusDistributionMode=::NewEnumerator1
    Sphere Radius=&Multiply_Float(A=5, B=$System.ExplosionScale)
    U Position:NiagaraFloat=&ReturnNormalizedExecIndex(Normalized Index Scale=3.5)
    V Position:NiagaraFloat=&ReturnNormalizedExecIndex(Normalized Index Scale=1.7)
   )
   &AddVelocityFromPoint(
    Module.Use Falloff Distance=false
    Velocity Strength=&Multiply_Float(A=&UniformRangedFloat(Minimum=100, Maximum=200), B=$System.ExplosionScale)
   )
  ]
  particleUpdateScript [
   &ParticleState()
   &SolveForcesAndVelocity(Clamp Velocity=false, Limit Acceleration=false)
   &ScaleSpriteSize(Scale Factor=&Vector2DFromCurve(
     Vector2Curve=DataInterface() {propertyValues {
       XCurve {keys [<at 0 0.75 RCIM_Cubic>,<at 1 1 RCIM_Cubic>]}
       YCurve {keys [{arriveTangent=1.5305, interpMode=RCIM_Cubic, leaveTangent=1.5305, tangentMode=RCTM_User, time=0, value=1}, <at 1 2 RCIM_Cubic>]}
      }}
     ScaleCurve=<1 1>
    ))
   &Color(Color=&MakeLinearColorFromVectorAndFloat(
     "Vector (RGB)"=&MakeVectorFromLinearColorRGB(Color=&ColorFromCurve(ColorCurve=DataInterface() {propertyValues {
         AlphaCurve {keys [<at 0 1 RCIM_Cubic>]}
         BlueCurve {keys [<at 0 0 RCIM_Cubic>,<at 0.2913 0.4185>,<at 0.6003 0.0899 RCIM_Cubic>]}
         GreenCurve {keys [<at 0 10 RCIM_Cubic>,<at 0.2913 5>,<at 0.6003 0.5 RCIM_Cubic>]}
         RedCurve {keys [<at 0 5.8251 RCIM_Cubic>,<at 0.2913 0>,<at 0.6003 0 RCIM_Cubic>]}
        }}))
     "Float (Alpha)"=&FloatFromCurve(FloatCurve=DataInterface() {propertyValues {Curve {keys [<at 0 1 RCIM_Cubic>,<at 1 0>]}}})
    ))
  ]
  rendererClasses [&NiagaraSpriteRendererProperties]
  renderers [{rendererClass=&NiagaraSpriteRendererProperties}]
 }
}
summary {
 systemName=NS_BigBurst
 emitters [
  {emitterName=Streaks_Fat, rendererClasses [&NiagaraSpriteRendererProperties]}
  {emitterName=NE_Sprites, rendererClasses [&NiagaraSpriteRendererProperties]}
  {
   emitterName=Flash_Flare001
   rendererClasses [&NiagaraSpriteRendererProperties, &NiagaraLightRendererProperties]
  }
 ]
 userVariables [
  User.Color {type=LinearColor, description="", defaultValue {struct=LinearColor, value=<rgba 0.0283 1 0.1353 1>}}
  User.ExplosionDelay {type=NiagaraFloat, defaultValue=0, description=""}
  User.IsCharged {type=NiagaraBool, defaultValue=true}
  User.LoopDuration {type=NiagaraFloat, defaultValue=2, description=""}
 ]
}
