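# Niagara lightning burst effect for the gamepad in the GameBreaker emote.
# A single emitter spawns a few short-lived red-orange sprites with random size, rotation, sub-image, camera offset and material parameters.
# Over lifetime, a curve dips and restores a dynamic material parameter while a color curve spikes brightness early for a flash.
# User color and skeletal mesh parameters are exposed, but the emitter's shown modules do not read them.
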NiagaraSystem /BRCosmetics/Animation/Game/MainPlayer/Emotes/GameBreaker/FX/NS_GamePad_Lightning
dependencies {
 usedDataInterfaces [&NiagaraDataInterfaceCurve, &NiagaraDataInterfaceColorCurve]
 usedDynamicInputs []
 usedModules [&SystemState]
}
emitters {Lightning_Sprites {
  emitterUpdateScript [
   &EmitterState()
   &SpawnBurst_Instantaneous(Spawn Count=&UniformRangedInt(Minimum=6, Maximum=7), Use Spawn Probability=false)
  ]
  particleSpawnScript [
   /Niagara/Modules/Spawn/Initialization/InitializeParticle(
    Color=<rgba 4 0.0552 0 1>
    Lifetime=&UniformRangedFloat(Minimum=0.8, Maximum=0.3)
    Module.Write Color=true
    Module.Write Lifetime=true
    Module.Write Mass=true
    Module.Write Material Random=true
    Module.Write Position=false
    Module.Write Scale=false
    Module.Write SpriteRotation=true
    Module.Write SpriteSize=true
    Position=&SimulationPosition()
    Sprite Rotation=&UniformRangedFloat(Minimum=0, Maximum=360)
    Sprite Size=&Vector2DFromFloat(Value=&Lerp_Float(A=33, B=27, Alpha=&UniformRangedFloat(Minimum=0, Maximum=1)))
   )
   &DynamicMaterialParameters(
    Index 0 Param 1=&UniformRangedFloat(Minimum=0, Maximum=1)
    Index 0 Param 2=&UniformRangedFloat(Minimum=0.1, Maximum=0.1)
    Index 0 Param 3=1
    Module.Param0WriteEnabled=true
    Module.Param1WriteEnabled=true
    Module.Param2WriteEnabled=true
    Module.Param3WriteEnabled=false
   ) {module=DynamicMaterialParameters001}
   /SetVariables(Particles.SubImageIndex=&UniformRangedFloat(Minimum=0, Maximum=3)) {isSetParameters=true}
   &CameraOffset(Camera Offset Amount=&UniformRangedFloat(Minimum=-4, Maximum=8))
  ]
  particleUpdateScript [
   &ParticleState()
   &DynamicMaterialParameters(
    Index 0 Param 3=&FloatFromCurve(FloatCurve=DataInterface() {propertyValues {Curve {keys [<at 0 1 RCIM_Cubic>,<at 0.2 0.025 RCIM_Cubic>,<at 1 1>]}}})
    Module.Param0WriteEnabled=false
    Module.Param1WriteEnabled=false
    Module.Param2WriteEnabled=true
    Module.Param3WriteEnabled=false
   )
   &ScaleColor(
    Scale RGB=&Multiply_VectorByFloat(
     Vector=&MakeVectorFromLinearColorRGB(Color=&ColorFromCurve(ColorCurve=DataInterface() {propertyValues {
         AlphaCurve {keys []}
         BlueCurve {keys [<at 0 1>,<at 0.15 10 RCIM_Cubic>,<at 1 1 RCIM_Cubic>]}
         GreenCurve {keys [<at 0 1>, {arriveTangent=-0.1823, interpMode=RCIM_Cubic, leaveTangent=-0.1823, time=0.15, value=9.045}, <at 1 0.8177 RCIM_Cubic>]}
         RedCurve {keys [<at 0 1>, {arriveTangent=-0.1812, interpMode=RCIM_Cubic, leaveTangent=-0.1812, time=0.15, value=8.5417}, <at 1 0.8188 RCIM_Cubic>]}
        }}))
     Float=1
    )
    ScaleA=false
    ScaleRGB=true
   )
  ]
  rendererClasses [&NiagaraSpriteRendererProperties]
  renderers [{rendererClass=&NiagaraSpriteRendererProperties}]
 }}
summary {
 systemName=NS_GamePad_Lightning
 emitters [{emitterName=Lightning_Sprites, rendererClasses [&NiagaraSpriteRendererProperties]}]
 userVariables [
  User.Color {type=LinearColor, description="", defaultValue {struct=LinearColor, value=<rgba 0 0.8108 2 1>}}
  User.ColorChange {type=LinearColor, description="", defaultValue {struct=LinearColor, value=<rgba 1 0.0469 0.6556 1>}}
  User.Skeletal Mesh {
   type=NiagaraDataInterfaceSkeletalMesh
   defaultValue {
    struct=NiagaraExt_VariableValue_DataInterface
    value {dataInterface=&self:NiagaraDataInterfaceSkeletalMesh_1}
   }
  }
  User.SkeletalMesh {
   type=NiagaraDataInterfaceSkeletalMesh
   defaultValue {
    struct=NiagaraExt_VariableValue_DataInterface
    value {dataInterface=&self:NiagaraDataInterfaceSkeletalMesh_0}
   }
  }
  User.UseCustomColor {type=NiagaraFloat, defaultValue=0, description=""}
 ]
}
