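# Niagara effect for the GameBreaker emote that produces a brief, flickering red flare with a matching light.
# A single emitter bursts one short-lived red sprite, offset toward the camera.
# A float curve pulses its alpha through three peaks while RGB is boosted, and FN_Light_v2 feeds a light renderer alongside the sprite renderer.
# The exposed Secondary Color user parameter is not referenced by the modules shown.
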NiagaraSystem /BRCosmetics/Animation/Game/MainPlayer/Emotes/GameBreaker/FX/NS_RedLight
dependencies {usedDataInterfaces [&NiagaraDataInterfaceCurve], usedDynamicInputs [], usedModules [&SystemState]}
emitters {Flash_Flare001 {
  emitterUpdateScript [
   &EmitterState(Life Cycle Mode=::NewEnumerator1, Loop Behavior=::NewEnumerator1, Loop Duration=1, UseLoopDelay=false)
   &SpawnBurst_Instantaneous(Spawn Count=1, Use Spawn Probability=false)
  ]
  particleSpawnScript [
   /Niagara/Modules/Spawn/Initialization/InitializeParticle(
    Color=<rgba 1 0 0 0.2>
    Lifetime=0.5
    Module.Write Color=true
    Module.Write Lifetime=true
    Module.Write Mass=true
    Module.Write Material Random=true
    Module.Write Position=false
    Module.Write Scale=false
    Module.Write SpriteRotation=false
    Module.Write SpriteSize=true
    Position=&SimulationPosition()
    Sprite Size=<55 55>
   )
   &CameraOffset(Camera Offset Amount=25)
   &DynamicMaterialParameters(
    Index 0 Param 1=0.2
    Module.Param0WriteEnabled=true
    Module.Param1WriteEnabled=true
    Module.Param2WriteEnabled=true
    Module.Param3WriteEnabled=true
   )
  ]
  particleUpdateScript [
   &ParticleState()
   &SolveForcesAndVelocity(Clamp Velocity=false, Limit Acceleration=false)
   &ScaleColor(
    Scale Alpha=&FloatFromCurve(FloatCurve=DataInterface() {propertyValues {Curve {keys [<at 0 0>,<at 0.1405 1>,<at 0.3124 0.25>,<at 0.462 1>,<at 0.663 0.25>,<at 0.8036 1>,<at 1 0>]}}})
    Scale RGB=&VectorFromFloat(Value=2)
    ScaleA=true
    ScaleRGB=true
   )
   &FN_Light_v2(Light Exponent=3, Light Radius=150, LightEnabled=true, Scale Color=1)
  ]
  rendererClasses [&NiagaraSpriteRendererProperties, &NiagaraLightRendererProperties]
  renderers [
   {rendererClass=&NiagaraSpriteRendererProperties}
   {rendererClass=&NiagaraLightRendererProperties, rendererIndex=1}
  ]
 }}
summary {
 systemName=NS_RedLight
 emitters [{
   emitterName=Flash_Flare001
   rendererClasses [&NiagaraSpriteRendererProperties, &NiagaraLightRendererProperties]
  }]
 userVariables [User.Secondary Color {type=LinearColor, description="", defaultValue {struct=LinearColor, value=<rgba 1 1 1 1>}}]
}
